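# Niagara smoke puff for the Peaceful Poem emote: a short ground-level burst of expanding rings and dust.
# SmokeGroundRings spawns two flat, upward-facing sprites that grow and fade.
# Smoke_Dust001 throws rotating sprites outward with slight lift and drag, skipped when User.bIsInWater is set.
# Both tint by time of day; the water drop, foam, splash and streak emitters are present but disabled.
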NiagaraSystem /BRCosmetics/Effects/Fort_Effects/Emotes/PeacefulPoem/NS_Emote_PeacefulPoem_Smoke
dependencies {
 usedDataInterfaces [&NiagaraDataInterfaceCurve, &NiagaraDataInterfaceVector2DCurve, &NiagaraDataInterfaceColorCurve]
 usedDynamicInputs [
  &SimulationPosition
  /Game/Effects/Niagara/DynamicInputs/Random/FN_RandomVectorInCone
  &Boolean_OrOperation
  &SetBoolByFloatComparison
  &Abs_Float
  &MakeFloatFromVector
 ]
 usedModules [&SystemState, &Timeline]
}
emitters {
 FN_Water_Drops_ {
  bEnabled=false
  emitterUpdateScript [
   &EmitterState()
   &SpawnBurst_Instantaneous(
    Spawn Count=&MakeCustomIntFromBool(Bool=$User.bIsInWater, True INT=&RandomRangeInt(Minimum=25, Maximum=30), False INT=0)
    Use Spawn Probability=false
    UseLoopCountLimit=false
   )
  ]
  particleSpawnScript [
   /SetVariables(Particles.Random=&UniformRangedFloat(Minimum=0, Maximum=1)) {isSetParameters=true}
   &InitializeParticle(
    Color=<rgba 5 5 5 1>
    Lifetime=&UniformRangedFloat(Minimum=0.5, Maximum=0.75)
    Mass=&UniformRangedFloat(Minimum=0.6, Maximum=1.2)
    Mass Mode=::NewEnumerator1
    Sprite Rotation Mode=::NewEnumerator2
    Sprite Size=&Vector2DFromCurve(
     Vector2Curve=DataInterface() {propertyValues {
       XCurve {keys [<at 0 1.0924 RCIM_Cubic>,<at 0.9988 0.3489 RCIM_Cubic>]}
       YCurve {keys [
         {arriveTangent=-6.848, interpMode=RCIM_Cubic, leaveTangent=-6.848, tangentMode=RCTM_User, time=0.0012, value=4.1914}
         {arriveTangent=-2.264, interpMode=RCIM_Cubic, leaveTangent=-2.264, time=0.2419, value=3.2838}
         <at 0.9988 1.9328 RCIM_Cubic>
        ]}
      }}
     CurveIndex=$Particles.Random
     ScaleCurve=<2 2>
    )
    Sprite Size Mode=::NewEnumerator4
    UsePositionOffset=false
    Write Material Random=true
   )
   &DynamicMaterialParameters(
    Index 0 Param 1=&UniformRangedFloat(Minimum=4, Maximum=10)
    Module.Param0WriteEnabled=true
    Module.Param1WriteEnabled=false
    Module.Param2WriteEnabled=false
    Module.Param3WriteEnabled=false
   )
   &CylinderLocation(
    Cylinder Height=1
    Cylinder Radius=25
    Lathe Profile=DataInterface() {propertyValues {Curve {keys []}}}
    MaskSpawnGroup=false
    Module.Enable Lathe Profile=false
    Module.Override Randomness=false
    Module.Surface Only=true
    Surface Only Band Thickness=2
   )
   /SetVariables(Particles.SpriteSize=&Multiply_Vector2(
     A=$Particles.SpriteSize
     B=&Vector2DFromFloat(Value=&FloatFromCurve(
       FloatCurve=DataInterface() {propertyValues {Curve {keys [
           {arriveTangent=1.6964, interpMode=RCIM_Cubic, time=-0.0012, value=-0.0071}
           {arriveTangent=13.7683, interpMode=RCIM_Cubic, leaveTangent=13.7683, time=0.3287, value=17.2263}
           {arriveTangent=-3.3438, interpMode=RCIM_Cubic, leaveTangent=-3.3438, time=0.7139, value=9.8392}
           {arriveTangent=1.8535, interpMode=RCIM_Cubic, leaveTangent=1.8535, tangentMode=RCTM_User, time=0.9945, value=15}
          ]}}}
       CurveIndex=$Emitter.Scalability.DistanceFraction
      ))
    )) {enabled=false, isSetParameters=true}
   &AddVelocity(
    Cone Angle=180
    Cone Axis=<0 0 1>
    Module.UseConeFalloff=false
    Module.UseVelDistribution=false
    Override Seed=false
    Velocity Mode=::NewEnumerator2
    Velocity Speed=&RandomRangeFloat(Minimum=300, Maximum=500)
   )
  ]
  particleUpdateScript [
   &ParticleState()
   /Game/Effects/Niagara/Emitters/Water/FN_DistanceKillParticles(AgeReduiceAmount=0.01) {enabled=false}
   &ScaleColor(
    Scale Alpha=&FloatFromCurve(FloatCurve=DataInterface() {propertyValues {Curve {keys [
         <at 0 0 RCIM_Cubic>
         {arriveTangent=0.002, interpMode=RCIM_Cubic, leaveTangent=0.002, time=0.1183, value=1.0113}
         <at 1.0011 0.002 RCIM_Cubic>
        ]}}})
    ScaleA=true
    ScaleRGB=true
   )
   /Niagara/Modules/Update/Forces/CurlNoiseForce(
    Noise Frequency=1
    Noise Strength=&FloatFromCurve(FloatCurve=DataInterface() {propertyValues {Curve {keys [
         {arriveTangent=-1159.6794, interpMode=RCIM_Cubic, leaveTangent=-1159.6792, tangentMode=RCTM_User, time=-0.0046, value=500}
         {arriveTangent=-447.4286, interpMode=RCIM_Cubic, leaveTangent=-447.4286, time=0.2223, value=189.6793}
         <at 1.0011 50 RCIM_Cubic>
        ]}}})
    Pan Noise=false
   )
   &Drag(Use Linear Drag=true, Use Rotational Drag=false)
   &GravityForce()
   &SolveForcesAndVelocity(Clamp Velocity=false, Limit Acceleration=false)
  ]
  rendererClasses [&NiagaraSpriteRendererProperties]
  renderers [{rendererClass=&NiagaraSpriteRendererProperties}]
 }
 FN_Water_Foam_SnapToSurface001 {
  bEnabled=false
  emitterUpdateScript [
   &EmitterState()
   &SpawnBurst_Instantaneous(
    Spawn Count=&Multiply_Int(A=&MakeCustomIntFromBool(Bool=$User.bIsInWater, True INT=1, False INT=0), B=1)
    Use Spawn Probability=false
   )
  ]
  particleSpawnScript [
   &InitializeParticle(
    Lifetime=&UniformRangedFloat(Minimum=0.8, Maximum=1.2)
    Mass=&UniformRangedFloat(Minimum=0.6, Maximum=1.2)
    Mass Mode=::NewEnumerator1
    Mesh Scale=&VectorFromFloat(Value=&Multiply_Float(A=70, B=$User.EffectScale))
    Mesh Scale Mode=::NewEnumerator4
    UsePositionOffset=false
    Write Material Random=true
   )
   &InitialMeshOrientation(
    Orientation Vector=<0 0 0>
    Rotation=&UniformRangedVector(Minimum=<0 0 0>, Maximum=<0 0 1>)
    Use Orientation Vector=true
    Use Rotation Vector=true
   ) {module=InitialMeshOrientation001}
   &DynamicMaterialParameters(
    Index 1 Param 3=200
    Index 3 Param 1=$User.WaterBodyIndex
    Index 3 Param 2=$User.WaterDepth
    Module.Param0WriteEnabled1=false
    Module.Param0WriteEnabled3=true
    Module.Param1WriteEnabled1=false
    Module.Param1WriteEnabled3=true
    Module.Param2WriteEnabled1=true
    Module.Param2WriteEnabled3=true
    Module.Param3WriteEnabled1=false
    Module.Param3WriteEnabled3=true
    Write Parameter Index 0=false
    Write Parameter Index 1=true
    Write Parameter Index 3=true
   )
   &InitializeWaterSnapToOwner()
   &InheritVelocity(Module.Limit Speed=true)
   &CameraOffset(Camera Offset Amount=20)
  ]
  particleUpdateScript [
   &ParticleState()
   &DynamicMaterialParameters(
    Index 0 Param 1=&FloatFromCurve(FloatCurve=DataInterface() {propertyValues {Curve {keys [<at -0.0046 1>,<at 1.0011 4>]}}})
    Index 0 Param 2=&FloatFromCurve(FloatCurve=DataInterface() {propertyValues {Curve {keys [{arriveTangent=0.6218, interpMode=RCIM_Cubic, leaveTangent=0.6218, tangentMode=RCTM_User, time=-0.0046, value=0.5}, <at 1.0011 1 RCIM_Cubic>]}}})
    Index 0 Param 3=&FloatFromCurve(FloatCurve=DataInterface() {propertyValues {Curve {keys [<at -0.0046 1>,<at 1.0011 0.5>]}}})
    Index 2 Param 1=&FloatFromCurve(FloatCurve=DataInterface() {propertyValues {Curve {keys [<at -0.0046 0.4>,<at 1.0011 1.5>]}}})
    Index 2 Param 2=&FloatFromCurve(FloatCurve=DataInterface() {propertyValues {Curve {keys [<at -0.0046 0.1>,<at 0.11 2.6>,<at 1.0011 0.3>]}}})
    Index 2 Param 3=1
    Index 2 Param 4=&FloatFromCurve(FloatCurve=DataInterface() {propertyValues {Curve {keys [
         {arriveTangent=-1.9991, interpMode=RCIM_Cubic, time=0, value=1}
         {arriveTangent=-2.7143, interpMode=RCIM_Cubic, leaveTangent=-2.7143, tangentMode=RCTM_User, time=0.2095, value=0.6901}
         <at 1 0 RCIM_Cubic>
        ]}}})
    Module.Param0WriteEnabled=true
    Module.Param0WriteEnabled2=true
    Module.Param1WriteEnabled=true
    Module.Param1WriteEnabled2=true
    Module.Param2WriteEnabled=true
    Module.Param2WriteEnabled2=true
    Module.Param3WriteEnabled=true
    Module.Param3WriteEnabled2=true
    Write Parameter Index 2=true
   ) {module=DynamicMaterialParameters001}
   &Drag(Drag=2, Use Linear Drag=true, Use Rotational Drag=false)
   &SolveForcesAndVelocity(Clamp Velocity=false, Limit Acceleration=false)
   &ScaleMeshSize(Scale Factor=&VectorFromFloat(Value=&FloatFromCurve(FloatCurve=DataInterface() {propertyValues {Curve {keys [
          {arriveTangent=26.4511, interpMode=RCIM_Cubic, leaveTangent=26.451, tangentMode=RCTM_User, time=-0.0046, value=0.6073}
          {arriveTangent=3.3733, interpMode=RCIM_Cubic, leaveTangent=3.3733, time=0.1268, value=2.5016}
          <at 1.0011 4 RCIM_Cubic>
         ]}}}))) {module=MeshSizeScale}
   &ConstrainPositionToPlane(
    Plane Normal:Vector3f=$Particles.WaterPlaneNormal
    Plane Position:NiagaraPosition=&AddVectorToPosition(Position=$Particles.Position, Vector=&Add_Vector(A=$Particles.WaterPlaneLocation, B=<0 0 0>))
   )
  ]
  rendererClasses [&NiagaraMeshRendererProperties]
  renderers [{rendererClass=&NiagaraMeshRendererProperties, rendererIndex=0}]
 }
 FN_Water_splash_SnapToSurface {
  bEnabled=false
  emitterUpdateScript [
   &EmitterState()
   &SpawnBurst_Instantaneous(
    Spawn Count=&Multiply_Int(
     A=&MakeCustomIntFromBool(Bool=$User.bIsInWater, True INT=&RandomRangeInt(Minimum=1, Maximum=2), False INT=0)
     B=1
    )
    Use Spawn Probability=false
   )
  ]
  particleSpawnScript [
   &InitializeParticle(
    Lifetime=&UniformRangedFloat(Minimum=0.7, Maximum=0.9)
    Mass=&UniformRangedFloat(Minimum=0.6, Maximum=1.2)
    Mass Mode=::NewEnumerator1
    Mesh Scale=<0.3 0.3 0.25>
    Mesh Scale Mode=::NewEnumerator4
    UsePositionOffset=false
    Write Material Random=true
   )
   &InitialMeshOrientation(
    Mesh Orientation Mode=::NewEnumerator1
    Rotation=&UniformRangedVector(Minimum=<0 0 0>, Maximum=<0 0 360>)
    Use Rotation Vector=true
   )
   &CylinderLocation(
    Cylinder Height=1
    Cylinder Radius=10
    Lathe Profile=DataInterface() {propertyValues {Curve {keys []}}}
    MaskSpawnGroup=false
    Module.Enable Lathe Profile=false
    Module.Override Randomness=false
    Module.Surface Only=false
   )
   /SetVariables(Particles.Position=&Add_Vector(
     A=Linked() {linkedVariable {name=Particles.Position, type=NiagaraPosition}}
     B=&Multiply_Vector(A=$Engine.Owner.Velocity, B=<0.1 0.1 0>)
    )) {isSetParameters=true}
   &InheritVelocity(Module.Limit Speed=true, Velocity Limit=200, Velocity Scale=<0.5 0.5 0>)
   &InitializeWaterSnapToOwner()
  ]
  particleUpdateScript [
   &ParticleState()
   &DynamicMaterialParameters(
    Index 0 Param 1=&FloatFromCurve(FloatCurve=DataInterface() {propertyValues {Curve {keys [
         <at -0.0046 -2>
         {arriveTangent=1.9886, interpMode=RCIM_Cubic, leaveTangent=1.9886, time=0.7384, value=-0.7816}
         {arriveTangent=6.4587, interpMode=RCIM_Cubic, leaveTangent=6.4587, tangentMode=RCTM_User, time=1.0011, value=0}
        ]}}})
    Index 0 Param 2=&FloatFromCurve(FloatCurve=DataInterface() {propertyValues {Curve {keys [<at -0.0046 4>,<at 1.0011 0>]}}})
    Index 0 Param 3=&FloatFromCurve(FloatCurve=DataInterface() {propertyValues {Curve {keys [
         {arriveTangent=0.096, interpMode=RCIM_Cubic, leaveTangent=0.096, tangentMode=RCTM_User, time=-0.0046, value=0.0882}
         <at 1.0011 0.1372 RCIM_Cubic>
        ]}}})
    Index 0 Param 4=&FloatFromCurve(FloatCurve=DataInterface() {propertyValues {Curve {keys [
         {arriveTangent=1.9125, interpMode=RCIM_Cubic, leaveTangent=0.2728, tangentMode=RCTM_User, time=-0.0046, value=0.9276}
         {arriveTangent=4.6326, interpMode=RCIM_Cubic, leaveTangent=4.6326, tangentMode=RCTM_User, time=1.0011, value=2.7648}
        ]}}})
    Index 1 Param 1=&FloatFromCurve(FloatCurve=DataInterface() {propertyValues {Curve {keys [
         {interpMode=RCIM_Cubic, tangentMode=RCTM_User, time=0, value=0}
         {arriveTangent=3.1921, interpMode=RCIM_Cubic, leaveTangent=3.1921, tangentMode=RCTM_User, time=1, value=1}
        ]}}})
    Module.Index0Param0Write=true
    Module.Index0Param1Write=true
    Module.Index0Param2Write=true
    Module.Index0Param3Write=true
    Module.Index1Param0Write=true
    Module.Index1Param1Write=true
    Module.Index1Param2Write=true
    Module.Index1Param3Write=true
    Write Parameter Index 1=true
   )
   &ScaleColor(
    Scale Alpha=&FloatFromCurve(FloatCurve=DataInterface() {propertyValues {Curve {keys [<at -0.0046 1.0113 RCIM_Cubic>,<at 1.0011 0.002 RCIM_Cubic>]}}})
    ScaleA=true
    ScaleRGB=false
   )
   &Drag(Drag=2, Use Linear Drag=true, Use Rotational Drag=false)
   &SolveForcesAndVelocity(Clamp Velocity=false, Limit Acceleration=false)
   &ScaleMeshSize(Scale Factor=&VectorFromFloat(Value=&FloatFromCurve(FloatCurve=DataInterface() {propertyValues {Curve {keys [<at 0 1>,<at 1 1.2>]}}}))) {module=MeshSizeScale}
   &ConstrainPositionToPlane(
    Plane Normal:Vector3f=$Particles.WaterPlaneNormal
    Plane Position:NiagaraPosition=&AddVectorToPosition(Position=$Particles.Position, Vector=&Add_Vector(A=$Particles.WaterPlaneLocation, B=<0 0 0>))
   )
  ]
  rendererClasses [&NiagaraMeshRendererProperties]
  renderers [{rendererClass=&NiagaraMeshRendererProperties, rendererIndex=0}]
 }
 SmokeGroundRings {
  emitterUpdateScript [
   &EmitterState()
   &SpawnBurst_Instantaneous(Spawn Count=2, Use Spawn Probability=false, UseLoopCountLimit=false)
   &SpawnRate(
    SpawnRate=&Multiply_Float(
     A=&MakeCustomFloatFromBool(Bool=$User.bIsInWater, True Float=0, False Float=1)
     B=&FloatFromCurve(
      FloatCurve=DataInterface() {propertyValues {Curve {keys [
          {interpMode=RCIM_Cubic, tangentMode=RCTM_User, time=0, value=1}
          {arriveTangent=-5.661, interpMode=RCIM_Cubic, leaveTangent=-5.661, tangentMode=RCTM_User, time=0.5, value=0}
         ]}}}
      CurveIndex=$System.Timeline.NormalizedTime
      Scale Curve=4
     )
    )
    Use Spawn Probability=false
   ) {enabled=false}
  ]
  particleSpawnScript [
   &InitializeParticle(
    Color=<rgba 0.4427 0.4427 0.4427 0.35>
    Lifetime Max=0.35
    Lifetime Min=0.25
    Lifetime Mode=::NewEnumerator1
    Position Offset=<0 0 3>
    Sprite Rotation Mode=::NewEnumerator2
    Sprite Size Mode=::NewEnumerator3
    Uniform Sprite Size=&RandomRangeFloat(Minimum=2.5, Maximum=2.8)
    UsePositionOffset=true
    Write Material Random=true
   )
   &SystemLocation(MaskSpawnGroup=false, Offset=<0 50 5>) {enabled=false}
   &SpriteFacingAndAlignment(Sprite Facing=<0 0 1>, WriteAlignment=true, WriteFacing=true)
  ]
  particleUpdateScript [
   &ParticleState()
   &DynamicMaterialParameters(
    Index 0 Param 1=&FloatFromCurve(FloatCurve=DataInterface() {propertyValues {Curve {keys [<at 0 0.5>,<at 1 2>]}}})
    Module.Param0WriteEnabled=true
    Module.Param1WriteEnabled=false
    Module.Param2WriteEnabled=false
    Module.Param3WriteEnabled=false
   )
   &ScaleSpriteSize(
    Uniform Curve Scale=50
    Uniform Curve Sprite Scale=DataInterface() {propertyValues {Curve {keys [
        {arriveTangent=6.3591, interpMode=RCIM_Cubic, leaveTangent=6.3591, tangentMode=RCTM_User, time=0, value=0}
        {arriveTangent=0.0224, interpMode=RCIM_Cubic, leaveTangent=0.0224, tangentMode=RCTM_User, time=1, value=2}
       ]}}}
   )
   &CameraOffset(Camera Offset Amount=20)
   &ScaleColor(
    Scale Alpha=&FloatFromCurve(
     FloatCurve=DataInterface() {propertyValues {Curve {keys [
         {tangentMode=RCTM_User, time=0, value=1}
         {arriveTangent=-0.0405, interpMode=RCIM_Cubic, leaveTangent=-0.0405, tangentMode=RCTM_User, time=0.7009, value=1}
         {arriveTangent=-8.0089, interpMode=RCIM_Cubic, leaveTangent=-8.0089, tangentMode=RCTM_User, time=1, value=0}
        ]}}}
     Scale Curve=&FloatFromCurve(
      FloatCurve=DataInterface() {propertyValues {Curve {keys [
          {interpMode=RCIM_Cubic, tangentMode=RCTM_User, time=0, value=1}
          {arriveTangent=-11.478, interpMode=RCIM_Cubic, leaveTangent=-11.478, tangentMode=RCTM_User, time=0.25, value=0}
         ]}}}
      CurveIndex=$System.Timeline.NormalizedTime
     )
    )
    Scale RGB=&MakeVectorFromLinearColorRGB(Color=&FNTimeOfDayColor(
      Night Color=<rgba 0.2123 0.2489 0.27 1>
      Morning Color=<rgba 0.3626 0.4008 0.4271 1>
      Day Color=<rgba 0.3626 0.4008 0.4271 1>
      Evening Color=<rgba 0.3626 0.4008 0.4271 1>
     ))
    ScaleA=true
    ScaleRGB=true
   )
  ]
  rendererClasses [&NiagaraSpriteRendererProperties]
  renderers [{rendererClass=&NiagaraSpriteRendererProperties}]
 }
 Smoke_Dust001 {
  emitterUpdateScript [
   &EmitterState()
   &SpawnBurst_Instantaneous(
    Spawn Count=&MakeCustomIntFromBool(Bool=$User.bIsInWater, True INT=0, False INT=&RandomRangeInt(Minimum=25, Maximum=30))
    Use Spawn Probability=false
    UseLoopCountLimit=false
   )
  ]
  particleSpawnScript [
   /SetVariables(Particles.InitSpriteRotRate=&RandomRangeFloat(Minimum=-2, Maximum=2)) {isSetParameters=true}
   &InitializeParticle(
    Color Maximum=<rgba 0.2 0.1987 0.192 0.3>
    Color Minimum=<rgba 0.25 0.2432 0.2322 0.25>
    Color Mode=::NewEnumerator2
    Lifetime Max=0.5
    Lifetime Min=0.4
    Lifetime Mode=::NewEnumerator1
    Sprite Rotation Mode=::NewEnumerator2
    Sprite Size Mode=::NewEnumerator1
    Uniform Sprite Size Max=75
    Uniform Sprite Size Min=60
    UsePositionOffset=false
    Write Material Random=true
   )
   &ShapeLocation(
    Additional Quaternion Rotation=false
    Cylinder Height=0.5
    Cylinder Radius=15
    MaskSpawnGroup=false
    Override Seed=false
    Rotation Mode=::NewEnumerator3
    Shape Primitive=::NewEnumerator1
    Surface Only=false
    UseOwnerScale=true
   )
   &AddVelocity(
    Module.Use Falloff Distance=false
    Override Seed=false
    Velocity Mode=::NewEnumerator1
    Velocity Speed=&Lerp_Float(A=200, B=400, Alpha=$Particles.MaterialRandom)
   )
   &AddVelocity(Velocity=&RandomRangeVector(Minimum=<0 0 10>, Maximum=<0 0 40>)) {module=AddVelocity001}
   &InheritVelocity(Speed Limit=true, UseSpeedThreshold=false)
   &CameraOffset(Camera Offset Amount=10)
  ]
  particleUpdateScript [
   &ParticleState()
   &Drag(Drag=&RandomRangeFloat(Minimum=2, Maximum=3), Use Linear Drag=true, Use Rotational Drag=true)
   &GravityForce(Gravity=&RandomRangeVector(Minimum=<0 0 -250>, Maximum=<0 0 -275>)) {enabled=false}
   &SolveForcesAndVelocity(Clamp Velocity=false, Limit Acceleration=false)
   &SpriteRotationRate(Rotation Rate=&Multiply_Float(
     A=$Particles.InitSpriteRotRate
     B=&FloatFromCurve(
      FloatCurve=DataInterface() {propertyValues {Curve {keys [
          {arriveTangent=-4.0107, interpMode=RCIM_Cubic, leaveTangent=-4.0107, tangentMode=RCTM_User, time=-0.0046, value=1.0113}
          {arriveTangent=-0.2496, interpMode=RCIM_Cubic, leaveTangent=-0.2496, tangentMode=RCTM_User, time=0.5, value=0.16}
         ]}}}
      Scale Curve=250
     )
    ))
   &DynamicMaterialParameters(
    Index 0 Param 1=&Add_Float(
     A=&RandomRangeFloat(Minimum=0, Maximum=1)
     B=&FloatFromCurve(
      FloatCurve=DataInterface() {propertyValues {Curve {keys [
          <at 0 0 RCIM_Cubic>
          {arriveTangent=1.5947, interpMode=RCIM_Cubic, leaveTangent=1.5947, tangentMode=RCTM_User, time=0.2557, value=0.5257}
          <at 1 1 RCIM_Cubic>
         ]}}}
      CurveIndex=$Particles.Age
      Scale Curve=0.25
     )
    )
    Module.Index0Param0Write=true
    Module.Index0Param1Write=true
    Module.Index0Param2Write=true
    Module.Index0Param3Write=true
   ) {module=DynamicMatParams}
   &ScaleColor(
    Scale Alpha=&Multiply_Float(
     A=&FloatFromCurve(
      FloatCurve=DataInterface() {propertyValues {Curve {keys [
          {interpMode=RCIM_Cubic, tangentMode=RCTM_User, time=0, value=0}
          <at 0.15 1 RCIM_Cubic>
          {arriveTangent=-0.6929, tangentMode=RCTM_User, time=0.9, value=0}
         ]}}}
      Scale Curve=&FN_FrontEndFloatModifier(Gameplay=0.85, Front End=1)
     )
     B=&FloatFromCurve(
      FloatCurve=DataInterface() {propertyValues {Curve {keys [
          {interpMode=RCIM_Cubic, tangentMode=RCTM_User, time=0, value=0.8}
          {interpMode=RCIM_Cubic, tangentMode=RCTM_User, time=0.5, value=1}
          {tangentMode=RCTM_User, time=1, value=0.8}
         ]}}}
      CurveIndex=Linked() {linkedVariable {name=NPC.FortniteNPC.TimeOfDay_0-1, type=NiagaraFloat}}
     )
    )
    Scale RGB=&VectorFromFloat(Value=&FNTimeOfDayScalar(Night Scale=1.4, Morning Scale=1.2, Day Scale=1.1, Evening Scale=1.2))
    ScaleA=true
    ScaleRGB=true
   )
   &ScaleSpriteSize(Uniform Curve Sprite Scale=DataInterface() {propertyValues {Curve {keys [
        {arriveTangent=1.8095, interpMode=RCIM_Cubic, leaveTangent=1.8095, tangentMode=RCTM_User, time=0, value=0.5}
        {arriveTangent=0.1638, interpMode=RCIM_Cubic, leaveTangent=0.1638, tangentMode=RCTM_User, time=1.0463, value=1}
       ]}}})
  ]
  rendererClasses [&NiagaraSpriteRendererProperties]
  renderers [
   {rendererClass=&NiagaraSpriteRendererProperties}
   {rendererClass=&NiagaraSpriteRendererProperties, rendererIndex=1}
  ]
 }
 Streaks001 {
  bEnabled=false
  emitterUpdateScript [
   &EmitterState()
   &SpawnBurst_Instantaneous(
    Spawn Count=&MakeCustomIntFromBool(Bool=$User.bIsInWater, True INT=0, False INT=&RandomRangeInt(Minimum=10, Maximum=15))
    Use Spawn Probability=false
    UseLoopCountLimit=false
   )
  ]
  particleSpawnScript [
   &InitializeParticle(
    Color Maximum=<rgba 0.3 0.3 0.3 0.65>
    Color Minimum=<rgba 0.25 0.25 0.25 0.6>
    Color Mode=::NewEnumerator2
    Lifetime Max=0.35
    Lifetime Min=0.25
    Lifetime Mode=::NewEnumerator1
    Sprite Rotation Mode=::NewEnumerator2
    Sprite Size Max=<10 40>
    Sprite Size Min=<8 20>
    Sprite Size Mode=::NewEnumerator2
    UsePositionOffset=false
    Write Material Random=true
   )
   &CylinderLocation(Cylinder Height=0, Cylinder Radius=1, MaskSpawnGroup=false, Override Seed=false, UseOwnerScale=true)
   &AddVelocity(Module.Use Falloff Distance=false, Override Seed=false, Velocity Mode=::NewEnumerator1, Velocity Speed=300)
   &AddVelocity(Velocity=&RandomRangeVector(Minimum=<0 0 0>, Maximum=<0 0 100>)) {module=AddVelocity001}
   &ApplyInitialForces()
   &DynamicMaterialParameters(
    Index 0 Param 1=&UniformRangedFloat(Minimum=0, Maximum=1)
    Index 0 Param 2=&UniformRangedFloat(Minimum=1, Maximum=2)
    Index 0 Param 3=1
    Index 0 Param 4=1
    Module.Param0WriteEnabled=true
    Module.Param1WriteEnabled=true
    Module.Param2WriteEnabled=true
    Module.Param3WriteEnabled=true
   )
   &CameraOffset(Camera Offset Amount=30)
  ]
  particleUpdateScript [
   &ParticleState()
   &Drag(Use Linear Drag=true, Use Rotational Drag=true)
   &SolveForcesAndVelocity(Clamp Velocity=false, Limit Acceleration=false)
   &ScaleColor(
    Scale Alpha=&FloatFromCurve(
     FloatCurve=DataInterface() {propertyValues {Curve {keys [
         {arriveTangent=7.0675, interpMode=RCIM_Cubic, leaveTangent=7.0675, tangentMode=RCTM_User, time=0, value=0}
         {arriveTangent=0.002, interpMode=RCIM_Cubic, leaveTangent=0.002, time=0.5, value=1}
         {arriveTangent=-6.928, interpMode=RCIM_Cubic, leaveTangent=-6.928, tangentMode=RCTM_User, time=1.0011, value=0.002}
        ]}}}
     Scale Curve=&FloatFromCurve(
      FloatCurve=DataInterface() {propertyValues {Curve {keys [
          {interpMode=RCIM_Cubic, tangentMode=RCTM_User, time=0, value=0.8}
          {interpMode=RCIM_Cubic, tangentMode=RCTM_User, time=0.5, value=1}
          {tangentMode=RCTM_User, time=1, value=0.8}
         ]}}}
      CurveIndex=Linked() {linkedVariable {name=NPC.FortniteNPC.TimeOfDay_0-1, type=NiagaraFloat}}
     )
    )
    Scale RGB=&MakeVectorFromLinearColorRGB(Color=&ColorFromCurve(
      ColorCurve=DataInterface() {propertyValues {
        AlphaCurve {keys [<at 0 1 RCIM_Cubic>]}
        BlueCurve {keys [
          <at 0 1>
          {arriveTangent=-0.8476, interpMode=RCIM_Cubic, leaveTangent=-0.8476, time=0.25, value=0.1465}
          <at 0.5 0.5762>
          <at 0.75 0.147>
          <at 1 1 RCIM_Cubic>
         ]}
        GreenCurve {keys [
          <at 0 0.4342>
          {arriveTangent=0.5735, interpMode=RCIM_Cubic, leaveTangent=0.5735, time=0.25, value=0.2769}
          <at 0.5 0.7209>
          <at 0.75 0.2789>
          <at 1 0.4342 RCIM_Cubic>
         ]}
        RedCurve {keys [
          <at 0 0.2232>
          {arriveTangent=1.1473, interpMode=RCIM_Cubic, leaveTangent=1.1473, time=0.25, value=0.7031}
          <at 0.5 0.7969>
          <at 0.75 0.7011>
          <at 1 0.2232 RCIM_Cubic>
         ]}
       }}
      CurveIndex=Linked() {linkedVariable {name=NPC.FortniteNPC.TimeOfDay_0-1, type=NiagaraFloat}}
     ))
    ScaleA=true
    ScaleRGB=true
   )
  ]
  rendererClasses [&NiagaraSpriteRendererProperties]
  renderers [{rendererClass=&NiagaraSpriteRendererProperties}]
 }
}
summary {
 systemName=NS_Emote_PeacefulPoem_Smoke
 emitters [
  {emitterName=SmokeGroundRings, rendererClasses [&NiagaraSpriteRendererProperties]}
  {emitterName=Smoke_Dust001, rendererClasses [&NiagaraSpriteRendererProperties]}
  {bEnabled=false, emitterName=FN_Water_Drops_, rendererClasses [&NiagaraSpriteRendererProperties]}
  {bEnabled=false, emitterName=FN_Water_Foam_SnapToSurface001, rendererClasses [&NiagaraMeshRendererProperties]}
  {bEnabled=false, emitterName=Streaks001, rendererClasses [&NiagaraSpriteRendererProperties]}
  {bEnabled=false, emitterName=FN_Water_splash_SnapToSurface, rendererClasses [&NiagaraMeshRendererProperties]}
 ]
 userVariables [
  User.bIsFrontEnd {type=NiagaraBool}
  User.bIsFrontEndPreview {type=NiagaraBool}
  User.bIsInWater {type=NiagaraBool}
  User.EffectScale {type=NiagaraFloat, defaultValue=1, description=""}
  User.WaterBodyIndex {type=NiagaraFloat, defaultValue=0, description=""}
  User.WaterDepth {type=NiagaraFloat, defaultValue=0, description=""}
  User.WaterPlaneLocation {type=Vector3f, description="", defaultValue {struct=Vector3f, value=<0 0 0>}}
  User.WaterPlaneNormal {type=Vector3f, description="", defaultValue {struct=Vector3f, value=<0 0 0>}}
  User.WaveAttenuationFactor {type=NiagaraFloat, defaultValue=0, description=""}
  User.WaveTime {type=NiagaraFloat, defaultValue=0, description=""}
 ]
}
